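; Camera component that adapts a pilot cinematic camera's aspect ratio to the local player's viewport at runtime.
; While a hotfix scalable float enables it, cosmetic-only Tick waits for a local Fortnite player controller whose view target is the owner.
; It then calls UpdateAspectRatio once and disables ticking.
; That event scales the CineCamera filmback sensor width by viewport width over height and turns off the aspect ratio constraint.
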(event EventTick
  (params (DeltaSeconds :type Float%20%28single-precision%29))
  ;[0.0432 0 0.15 0.5] Running gametime logic to adjust the aspect ratio of the cinematic camera.
  (|CallParentFunction DeltaSeconds)
  (Networking|CanExecuteCosmeticEvents
    (:True
      (if (> (ScalableFloat|GetCachedScalableFloatValue (Variables|Default|GetHotfix_EnableDynamicAspectRatioAdjustment)) 0)
        (Utilities|GetCameraComponent
          (:Is%20Valid
            (for _array_element (Fortnite|GetLocalFortPlayerControllers true)
              (if (== (Pawn|GetViewTarget _array_element) (Components|GetOwner))
                (Utilities|FlowControl|DoOnce false
                  (:Completed
                    (CallFunction|UpdateAspectRatio _array_element)
                    (Components|Tick|SetComponentTickEnabled false)))))))
        (else
          _)))))

(event Custom|UpdateAspectRatio
  (params (Target :type Player%20Controller%20Object%20Reference))
  (bind (_sizex _sizey) (HUD|GetViewportSize Target))
  ;[0.0432 0 0.15 0.5] Running gametime logic to adjust the aspect ratio of the cinematic camera.
  (Networking|CanExecuteCosmeticEvents
    (:True
      (bind _cameracomponent (Utilities|GetCameraComponent)
        (:Is%20Valid
          (Utilities|Struct|SetmembersinCameraFilmbackSettings (Variables|/Script/CinematicCamera.CineCameraComponent|GetFilmback _cameracomponent) (* (Utilities|Struct|BreakCameraFilmbackSettings (Variables|/Script/CinematicCamera.CineCameraComponent|GetFilmback _cameracomponent)) (Math|Float|SafeDivide (Math|Conversions|ToFloat(Integer) _sizex) (Math|Conversions|ToFloat(Integer) _sizey))) (.SensorHeight (Utilities|Struct|BreakCameraFilmbackSettings (Variables|/Script/CinematicCamera.CineCameraComponent|GetFilmback _cameracomponent))))
          (Camera|SetConstraintAspectRatio _cameracomponent false))))))
