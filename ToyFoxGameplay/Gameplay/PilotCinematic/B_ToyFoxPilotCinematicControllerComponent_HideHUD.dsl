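; Player controller component that hides HUD elements for the owning player during the ToyFox pilot cinematic sequence.
; BeginPlay hides the elements and EndPlay restores them, so the component's lifetime defines the hidden window.
; SetHideHUDElements applies the configured element list under a ToyFoxPilotTeleport context, only where cosmetic events can execute.
; A custom DestroyComponent event lets the component destroy itself on authority.
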(event EventBeginPlay
  ;[0 0.15 0.01019 0.5] BeginPlay: Hide HUD elements for cinematic sequence.
  (|CallParentFunction)
  (Fortnite|FortShippingLog self (Utilities|String|Append "BeginPlay: Added component for player " (Fortnite|GetPlayerNameSafe (Components|GetOwner))))
  (Default|SetHideHUDElements true))

(event EventEndPlay
  (params (EndPlayReason :type EEndPlayReason%20Enum :default "Destroyed"))
  ;[0.1146 0.07428 0 0.5] EndPlay: Unhide HUD elements.
  (Utilities|FlowControl|Sequence
    (:then_0
      (Fortnite|FortShippingLog self (Utilities|String|Append "EndPlay: Removed component for player " (Fortnite|GetPlayerNameSafe (Components|GetOwner))))
      (Default|SetHideHUDElements false))
    (:then_1
      (|CallParentFunction EndPlayReason))))

(event Custom|DestroyComponent
  ;[0 0.15 0.07467 0.5] DestroyComponent: Invoke the K2_DestroyComponent function from within so it executes as expected.
  (Networking|SwitchHasAuthority
    (:Authority
      (Components|DestroyComponent))))

(fn SetHideHUDElements
  (params (bHideElements :type Boolean))
  ;[0 0.15 0.07106 0.5] SetHideHudElements: Hiding HUD elements during the entire cinematic sequence for owning player.
  (Networking|CanExecuteCosmeticEvents
    (:True
      (bind _asfort_player_controller (Utilities|Casting|CastToFortPlayerController (Components|GetOwner)))
      (HUD|SetHUDElementVisibilitywithContext _asfort_player_controller (Variables|Settings|GetHUDElementsToHide) bHideElements "ToyFoxPilotTeleport" "ToyFoxPilotTeleport")
      (Fortnite|FortShippingLog self (Utilities|String|Append "SetHideHUDElements: Set to " (Utilities|String|ToString(Boolean) bHideElements) " for player " (Fortnite|GetPlayerNameSafe _asfort_player_controller))))))
