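# Master Level Sequence for the ToyFox pilot cinematic, layering Gameplay, Animation, SFX, TechArt, VFX, World and EditorOnly subsequences.
# A camera cut track switches from the Splash cine camera to the Cockpit camera, both attached and animated for focal length.
# The Cockpit camera stacks transform sections for main path, jitter and offset, plus focus pulls.
# Spawn tracks manage hidden root cube and mech actors, while fade and MPC_Toyfox_SequenceControl tracks drive screen fade, visor alpha and flickering eye glow.
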LevelSequence /ToyFoxGameplay/Gameplay/PilotCinematic/ToyFoxPilotCinematicLevelSequence
bindings [
 CameraComponent {
  id=#1
  parent_id=#2
  possessed_class=CineCameraComponent
  tracks [
   {
    type=MovieSceneFloatTrack
    display_name=Current Focal Length
    sections [{
      type=MovieSceneFloatSection
      channels [None {
        type=MovieSceneScriptingFloatChannel
        default=10
        thinned_from=8
        keys [[171,52.6397],[187,50],[192,22],[197,24],[238,24],[244,14.5083],[249,15]]
       }]
     }]
   }
   {
    type=MovieSceneFloatTrack
    display_name="Manual Focus Distance (Focus Settings)"
    sections [{
      type=MovieSceneFloatSection
      channels [None {type=MovieSceneScriptingFloatChannel, default=1000, thinned_from=8, keys [[171,250],[239,250],[244,25187.5],[280,25187.5],[409,100000]]}]
     }]
   }
   {
    type=MovieSceneFloatTrack
    display_name=Current Aperture
    sections [{type=MovieSceneFloatSection, channels [None {type=MovieSceneScriptingFloatChannel, default=2.8, const [2.8,5]}]}]
   }
  ]
 }
 CineCameraActor_Splash {
  id=#3
  possessed_class=CineCameraActor
  tracks [
   {
    type=MovieScene3DTransformTrack
    sections [{
      type=MovieScene3DTransformSection
      channels [
       Location.X {type=MovieSceneScriptingDoubleChannel, default=0}
       Location.Y {type=MovieSceneScriptingDoubleChannel, default=0}
       Location.Z {type=MovieSceneScriptingDoubleChannel, default=50000}
       Rotation.X {type=MovieSceneScriptingDoubleChannel, default=0}
       Rotation.Y {type=MovieSceneScriptingDoubleChannel, default=0}
       Rotation.Z {type=MovieSceneScriptingDoubleChannel, default=0}
       Scale.X {type=MovieSceneScriptingDoubleChannel, default=1}
       Scale.Y {type=MovieSceneScriptingDoubleChannel, default=1}
       Scale.Z {type=MovieSceneScriptingDoubleChannel, default=1}
      ]
     }]
   }
   {type=MovieSceneSpawnTrack, sections [{type=MovieSceneSpawnSection, channels [None {type=MovieSceneScriptingBoolChannel, default=true}]}]}
  ]
 }
 ToyFoxPilotCinematicRootCube {
  id=#4
  possessed_class=FortStaticMeshActor
  tracks [
   {
    type=MovieScene3DTransformTrack
    sections [{
      type=MovieScene3DTransformSection
      channels [
       Location.X {type=MovieSceneScriptingDoubleChannel, default=0}
       Location.Y {type=MovieSceneScriptingDoubleChannel, default=0}
       Location.Z {type=MovieSceneScriptingDoubleChannel, default=50000}
       Rotation.X {type=MovieSceneScriptingDoubleChannel, default=0}
       Rotation.Y {type=MovieSceneScriptingDoubleChannel, default=0}
       Rotation.Z {type=MovieSceneScriptingDoubleChannel, default=270}
       Scale.X {type=MovieSceneScriptingDoubleChannel, default=1}
       Scale.Y {type=MovieSceneScriptingDoubleChannel, default=1}
       Scale.Z {type=MovieSceneScriptingDoubleChannel, default=1}
      ]
     }]
   }
   {type=MovieSceneSpawnTrack, sections [{type=MovieSceneSpawnSection, channels [None {type=MovieSceneScriptingBoolChannel, default=true}]}]}
   {
    type=MovieSceneVisibilityTrack
    sections [{type=MovieSceneVisibilitySection, channels [None {type=MovieSceneScriptingBoolChannel, default=false, const [false,1]}]}]
   }
  ]
 }
 CameraComponent {
  id=#5
  parent_id=#3
  possessed_class=CineCameraComponent
  tracks [
   {
    type=MovieSceneFloatTrack
    display_name=Current Focal Length
    sections [{
      type=MovieSceneFloatSection
      channels [None {
        type=MovieSceneScriptingFloatChannel
        default=35.2
        keys [[75,44],[79,43],[84,30],[88,30.6],[92,30],[127,30],[131,38],[135,34.6],[138,35.2],[141,35]]
       }]
     }]
   }
   {
    type=MovieSceneFloatTrack
    display_name="Manual Focus Distance (Focus Settings)"
    sections [{type=MovieSceneFloatSection, channels [None {type=MovieSceneScriptingFloatChannel, default=189.9787}]}]
   }
   {
    type=MovieSceneFloatTrack
    display_name=Current Aperture
    sections [{type=MovieSceneFloatSection, channels [None {type=MovieSceneScriptingFloatChannel, default=2.8}]}]
   }
   {type=MovieScene3DAttachTrack, sections [{type=MovieScene3DAttachSection, end_frame=186, start_frame=1}]}
  ]
 }
 CineCameraActor_Cockpit {
  id=#2
  possessed_class=CineCameraActor
  tracks [
   {
    type=MovieScene3DTransformTrack
    sections [
     {
      type=MovieScene3DTransformSection
      channels [
       Location.X {
        type=MovieSceneScriptingDoubleChannel
        default=-3280
        thinned_from=12
        keys [[204.2138,1116.1371],[238,1116.1371],[244,1313.4073],[280,1313.4073],[297,1365.7577],[317,1480.1972],[339,1205.0508],[366,-440.3962],[426,-3262.0083],[442,-3280]]
       }
       Location.Y {
        type=MovieSceneScriptingDoubleChannel
        default=1200
        thinned_from=12
        keys [[204.2138,-141],[238,-141],[244,-131.3599],[280,-131.3599],[297,-65.4478],[317,328.3359],[339,1029.7797],[366,1471.2674],[426,1210.4362],[442,1200]]
       }
       Location.Z {
        type=MovieSceneScriptingDoubleChannel
        default=410
        thinned_from=12
        keys [[204.2138,-715.0975],[238,-715.0975],[241,-716.6427],[244,-680.1879],[280,-680.1879],[297,-678.0788],[317,-597.6864],[339,-140.3612],[366,843.0171],[426,508.4062],[442,410]]
       }
       Rotation.X {
        type=MovieSceneScriptingDoubleChannel
        default=0
        thinned_from=12
        keys [[204.2138,1.3487],[238,1.3487],[244,0.1739],[280,0.1739],[297,0.1564],[317,0.0869],[339,0],[442,0]]
       }
       Rotation.Y {
        type=MovieSceneScriptingDoubleChannel
        default=0
        thinned_from=12
        keys [[204.2138,26],[238,26],[244,3.1811],[280,3.1811],[297,2.2735],[317,-2.9814],[339,-12.9052],[366,-18.6833],[426,-1.6013],[442,0]]
       }
       Rotation.Z {
        type=MovieSceneScriptingDoubleChannel
        default=360
        thinned_from=12
        keys [[204.2138,183],[280,183.1726],[297,187.2112],[317,206.1118],[339,248.5144],[366,304.9017],[426,358.8028],[442,360]]
       }
      ]
     }
     {
      type=MovieScene3DTransformSection
      channels [
       Location.X {type=MovieSceneScriptingDoubleChannel, default=0, const [0,22]}
       Location.Y {type=MovieSceneScriptingDoubleChannel, default=0, const [0,22]}
       Location.Z {type=MovieSceneScriptingDoubleChannel, default=0, const [0,22]}
       Rotation.X {
        type=MovieSceneScriptingDoubleChannel
        default=0.6206
        keys [[182,0],[191,0.2606],[193,-1.2],[196,1.5625],[199,-1],[204.2138,0.7],[210,-0.1301],[219,0.0301],[225,-0.1],[227,2.5],[232,-1],[238,0],[246,1.1],[254,0.3],[259,0.5289],[264,0.4966],[271,0.5228],[280,0.4954],[289,0.6206],[300,0.5181],[426,0],[442,0]]
       }
       Rotation.Y {
        type=MovieSceneScriptingDoubleChannel
        default=0.1
        thinned_from=23
        keys [[182,0],[191,0],[193,1.8243],[196,1.8796],[199,-0.0293],[203,0.5797],[206,0],[238,0],[241,0.1345],[246,0.4251],[252,-0.0383],[258,0.1014],[264,-0.0342],[271,0],[442,0]]
       }
       Rotation.Z {type=MovieSceneScriptingDoubleChannel, default=0, thinned_from=22, keys [[182,0],[442,0]]}
       Scale.X {type=MovieSceneScriptingDoubleChannel, default=0, const [0,22]}
       Scale.Y {type=MovieSceneScriptingDoubleChannel, default=0, const [0,22]}
       Scale.Z {type=MovieSceneScriptingDoubleChannel, default=0, const [0,22]}
      ]
     }
     {
      type=MovieScene3DTransformSection
      channels [
       Location.X {type=MovieSceneScriptingDoubleChannel, default=0, const [0,4]}
       Location.Y {type=MovieSceneScriptingDoubleChannel, default=-6, thinned_from=4, keys [[238,-6],[292,-6],[311,0]]}
       Location.Z {type=MovieSceneScriptingDoubleChannel, default=0, const [0,4]}
       Rotation.X {type=MovieSceneScriptingDoubleChannel, default=0.6, keys [[238,0],[246,0.6],[292,0.6],[311,0]]}
       Rotation.Y {type=MovieSceneScriptingDoubleChannel, default=0, thinned_from=4, keys [[238,0],[311,0]]}
       Rotation.Z {type=MovieSceneScriptingDoubleChannel, default=-1, keys [[238,-1.4],[246,-1],[292,-1],[311,0]]}
       Scale.X {type=MovieSceneScriptingDoubleChannel, default=0, const [0,4]}
       Scale.Y {type=MovieSceneScriptingDoubleChannel, default=0, const [0,4]}
       Scale.Z {type=MovieSceneScriptingDoubleChannel, default=0, const [0,4]}
      ]
     }
    ]
   }
   {
    type=MovieSceneSpawnTrack
    sections [{type=MovieSceneSpawnSection, channels [None {type=MovieSceneScriptingBoolChannel, default=false, keys [[83,false],[166,true]]}]}]
   }
   {type=MovieScene3DAttachTrack, sections [{type=MovieScene3DAttachSection, end_frame=442, start_frame=170}]}
  ]
 }
 ToyFoxPilotCinematicMechRoot {
  id=#6
  possessed_class=B_ToyFoxPilotCinematic_MechRoot_C
  tracks [
   {
    type=MovieSceneSpawnTrack
    sections [{type=MovieSceneSpawnSection, channels [None {type=MovieSceneScriptingBoolChannel, default=true, keys [[1,false],[95,true]]}]}]
   }
   {
    type=MovieScene3DTransformTrack
    sections [{
      type=MovieScene3DTransformSection
      channels [
       Location.X {type=MovieSceneScriptingDoubleChannel, default=0}
       Location.Y {type=MovieSceneScriptingDoubleChannel, default=0}
       Location.Z {type=MovieSceneScriptingDoubleChannel, default=0}
       Rotation.X {type=MovieSceneScriptingDoubleChannel, default=0}
       Rotation.Y {type=MovieSceneScriptingDoubleChannel, default=0}
       Rotation.Z {type=MovieSceneScriptingDoubleChannel, default=0}
       Scale.X {type=MovieSceneScriptingDoubleChannel, default=1}
       Scale.Y {type=MovieSceneScriptingDoubleChannel, default=1}
       Scale.Z {type=MovieSceneScriptingDoubleChannel, default=1}
      ]
     }]
   }
   {
    type=MovieSceneVisibilityTrack
    sections [{type=MovieSceneVisibilitySection, channels [None {type=MovieSceneScriptingBoolChannel, default=false, const [false,1]}]}]
   }
  ]
 }
]
display_rate {denominator=1, numerator=30}
tick_resolution {denominator=1, numerator=24000}
tracks [
 {type=MovieSceneSubTrack, display_name=Gameplay, sections [{type=MovieSceneSubSection, end_frame=601, start_frame=1}]}
 {type=MovieSceneSubTrack, display_name=Animation, sections [{type=MovieSceneSubSection, end_frame=601, start_frame=1}]}
 {type=MovieSceneSubTrack, display_name=SFX, sections [{type=MovieSceneSubSection, end_frame=601, start_frame=1}]}
 {type=MovieSceneSubTrack, display_name=TechArt, sections [{type=MovieSceneSubSection, end_frame=601, start_frame=1}]}
 {type=MovieSceneSubTrack, display_name=VFX, sections [{type=MovieSceneSubSection, end_frame=601, start_frame=1}]}
 {type=MovieSceneSubTrack, display_name=World, sections [{type=MovieSceneSubSection, end_frame=601, start_frame=1}]}
 {type=MovieSceneSubTrack, display_name=EditorOnly, sections [{type=MovieSceneSubSection, end_frame=660, start_frame=1}]}
 {
  type=MovieSceneFadeTrack
  sections [{type=MovieSceneFadeSection, channels [None {type=MovieSceneScriptingFloatChannel, default=0, keys [[75,1],[76,0]]}]}]
 }
 {
  type=MovieSceneMaterialParameterCollectionTrack
  display_name=MPC_Toyfox_SequenceControl
  sections [{
    type=MovieSceneParameterSection
    channels [
     VisorAlpha {type=MovieSceneScriptingFloatChannel, default=1, keys [[166,1],[167,0],[239,0],[241,1]]}
     EyeGlowIntensity {
      type=MovieSceneScriptingFloatChannel
      default=25
      keys [[1,0.2],[245,0.2],[249,4],[251,1],[252,25],[258,9],[260,0.5],[263,30],[264,50],[269,1],[273,60],[274,100],[276,1],[277,150],[287,20],[289,100],[301,100],[317,25],[330,1]]
     }
    ]
   }]
 }
 {
  type=MovieSceneCameraCutTrack
  sections [{type=MovieSceneCameraCutSection, end_frame=182, start_frame=1}, {type=MovieSceneCameraCutSection, end_frame=442, start_frame=182}]
 }
]
