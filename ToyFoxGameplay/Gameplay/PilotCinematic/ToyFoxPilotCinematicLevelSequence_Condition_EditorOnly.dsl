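; Movie Scene condition for the ToyFox pilot cinematic Level Sequence that gates content to the editor only.
; Its OnEvaluateCondition breaks the condition context and returns the result of the InEditor check.
; The intent is to pass only in the editor, excluding even Play-In-Editor sessions.
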(fn OnEvaluateCondition
  (params (ConditionContext :type Movie%20Scene%20Condition%20Context%20Structure%20%28by%20ref%29))
  ;[0.02205 0.1291 0.15 0.5] Only run in editor. Not even PIE.
  (return (Editor|InEditor (Utilities|Struct|BreakMovieSceneConditionContext ConditionContext))))
